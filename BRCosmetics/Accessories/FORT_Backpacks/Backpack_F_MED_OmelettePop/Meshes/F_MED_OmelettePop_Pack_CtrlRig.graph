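# Control Rig for the OmelettePop chicken backpack that procedurally adjusts the belt, shoulders and head sides as the body squashes.
# It measures squash as the distance between offset points following body and pelvis_01, compared to a rest distance.
# That difference pushes belt_all vertically and spreads side bones, while a weighted aim blends belt rotation between body and pelvis.
# Debug flags spawn Body and Pelvis test controls or display squash values.
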ControlRigBlueprint /BRCosmetics/Accessories/FORT_Backpacks/Backpack_F_MED_OmelettePop/Meshes/F_MED_OmelettePop_Pack_CtrlRig
graphs [
 Multiply_3_ContainedGraph {nodes [
   n1 = RigVMFunctionEntryNode()
   RigVMFunctionReturnNode(Result=&RigVMFunction_MathDoubleMul(B=@n1.C, A=&RigVMFunction_MathDoubleMul(A=@n1.A, B=@n1.B).Result).Result)
  ]}
 RigVMFunctionLibrary {links [], nodes []}
 RigVMFunction_Sequence_1_ContainedGraph {nodes [
   n1 = RigVMFunctionEntryNode()
   RigVMFunction_Sequence = &RigVMFunction_Sequence(ExecuteContext=@n1.ExecuteContext)
   n2 = &RigVMFunction_Sequence(ExecuteContext=@RigVMFunction_Sequence.B)
   n3 = &RigVMFunction_Sequence(ExecuteContext=@n2.B)
   RigVMFunctionReturnNode(A=@RigVMFunction_Sequence.A, B=@n2.A, C=@n3.A, D=@n3.B)
  ]}
 RigVMModel {nodes [
   RigUnit_BeginExecution = &RigUnit_BeginExecution()
   RigVMFunction_Sequence = RigVMAggregateNode(ExecuteContext=@RigUnit_BeginExecution.ExecutePin)
   n1 = &RigUnit_GetTransform(Item=(Type=Bone,Name="body"), bInitial=true)
   Add = /Script/RigVM.RigVMFunction_MathVectorAdd(B=(X=0,Y=0,Z=5), A=@n1.Transform.Translation) {is=struct RigVMUnitNode, template="Add::Execute(in A,in B,out Result)"}
   Multiply = &RigVMFunction_MathTransformMul(
    A=&RigVMFunction_MathTransformMakeRelative(Global=(Rotation=(X=0,Y=0,Z=0,W=1),Translation=(X=0,Y=0,Z=0),Scale3D=(X=1,Y=1,Z=1)), Global.Translation=@Add.Result, Parent=@n1.Transform).Local
    B=&RigUnit_GetTransform(Item=(Type=Bone,Name="body"), bInitial=false).Transform
   ) {template="Multiply::Execute(in A,in B,out Result)"}
   RigVMVariableNode(Value=@Multiply.Result, ExecuteContext=@RigVMFunction_Sequence.B) {variable=Squash End A}
   n2 = &RigVMFunction_MathTransformMul(
    A=&RigVMFunction_MathTransformMakeRelative(
     Global=(Rotation=(X=0,Y=0,Z=0,W=1),Translation=(X=0,Y=0,Z=0),Scale3D=(X=1,Y=1,Z=1))
     Global.Translation=&RigVMFunction_MathVectorAdd(B=(X=0,Y=0,Z=-5), A=@n1.Transform.Translation) {template="Add::Execute(in A,in B,out Result)"}.Result
     Parent=&RigUnit_GetTransform(Item=(Type=Bone,Name="pelvis_01"), bInitial=true).Transform
    ).Local
    B=&RigUnit_GetTransform(Item=(Type=Bone,Name="pelvis_01"), bInitial=false).Transform
   ) {template="Multiply::Execute(in A,in B,out Result)"}
   > RigVMVariableNode(Value=@n2.Result) {variable=Squash End B}
   RigVMFunction_MathVectorDistance = &RigVMFunction_MathVectorDistance(A=@Multiply.Result.Translation, B=@n2.Result.Translation)
   > n3 = RigVMVariableNode(Value=@RigVMFunction_MathVectorDistance.Result) {variable=Squash Distance}
   n4 = &RigVMFunction_ControlFlowBranch(ExecuteContext=@n3.ExecuteContext, Condition=RigVMVariableNode() {variable=Debug Squash Values}.Value)
   RigVMFunction_DebugTransformMutableNoSpace = /Script/RigVM.RigVMFunction_DebugTransformMutableNoSpace(
    Color=(R=1,G=0,B=0,A=1)
    DebugDrawSettings=(DepthPriority=SDPG_Foreground,Lifetime=-1)
    Mode=Axes
    Scale=10
    Thickness=1
    WorldOffset=(Rotation=(X=0,Y=0,Z=0,W=1),Translation=(X=0,Y=0,Z=0),Scale3D=(X=1,Y=1,Z=1))
    bEnabled=True
    ExecutePin=@n4.True
    Transform=RigVMVariableNode() {variable=Squash End A}.Value
   ) {is=struct RigVMUnitNode}
   > &RigVMFunction_DebugTransformMutableNoSpace(
    Color=(R=1,G=0,B=0,A=1)
    DebugDrawSettings=(DepthPriority=SDPG_Foreground,Lifetime=-1)
    Mode=Axes
    Scale=10
    Thickness=1
    WorldOffset=(Rotation=(X=0,Y=0,Z=0,W=1),Translation=(X=0,Y=0,Z=0),Scale3D=(X=1,Y=1,Z=1))
    bEnabled=True
    Transform=RigVMVariableNode() {variable=Squash End B}.Value
   )
   > RigVMDispatchNode(
    Enabled=True
    Prefix="Squash Dist: "
    ScreenColor=(R=1,G=1,B=1,A=1)
    ScreenDuration=0.05
    Value=RigVMVariableNode() {variable=Squash Distance}.Value
   ) {template="DISPATCH_RigVMDispatch_Print(in Prefix,in Value,in Enabled,in ScreenDuration,in ScreenColor)"}
   RigVMFunction_ControlFlowBranch = &RigVMFunction_ControlFlowBranch(Condition=RigVMVariableNode() {variable=Debug Squash Controls}.Value, ExecuteContext=@RigVMFunction_Sequence.A)
   &RigUnit_SetTransform(
    Item=(Type=Bone,Name="body")
    Weight=1
    bInitial=False
    Value=&RigUnit_GetTransform(Item=(Type=Control,Name="Body"), bInitial=False).Transform
    ExecutePin=@RigVMFunction_ControlFlowBranch.True
   )
   RigUnit_GetTransform = &RigUnit_GetTransform(Item=(Type=Control,Name="Pelvis"), bInitial=False)
   > RigUnit_SetTransform = &RigUnit_SetTransform(Item=(Type=Bone,Name="pelvis_MSTR"), Weight=1, bInitial=False, Value=@RigUnit_GetTransform.Transform)
   RigVMVariableNode(Value=0.5, ExecuteContext=@RigVMFunction_Sequence.C) {variable=Belt Push Percent}
   > RigVMVariableNode(Value=0.4) {variable=Belt Rotate Percent}
   Subtract = &RigVMFunction_MathDoubleSub(B=RigVMVariableNode() {variable=Original Distance}.Value, A=RigVMVariableNode() {variable=Squash Distance}.Value)
   Multiply_2 = RigVMAggregateNode(C=-1, A=@Subtract.Result, B=RigVMVariableNode() {variable=Belt Push Percent}.Value)
   > &RigUnit_OffsetTransformForItem(
    Item=(Type=Bone,Name="belt_all")
    OffsetTransform=(Rotation=(X=0,Y=0,Z=0,W=1),Translation=(X=0,Y=0,Z=0),Scale3D=(X=1,Y=1,Z=1))
    Weight=1
    bPropagateToChildren=True
    OffsetTransform.Translation.Z=@Multiply_2.Result
   )
   n5 = &RigUnit_GetTransform(Item=(Type=Bone,Name="body"), bInitial=False)
   n6 = /Script/RigVM.RigVMFunction_MathTransformRotateVector(Vector=(X=-1,Y=0,Z=0), Transform=@n5.Transform) {is=struct RigVMUnitNode, template="Rotate Vector::Execute(in Transform,in Vector,out Result)"}
   n7 = &RigVMFunction_MathDoubleSub(A=1, B=RigVMVariableNode() {variable=Belt Rotate Percent}.Value)
   n8 = &RigUnit_GetTransform(Item=(Type=Bone,Name="pelvis_01"), bInitial=False)
   n9 = &RigVMFunction_MathTransformRotateVector(Vector=(X=0,Y=-1,Z=0), Transform=@n8.Transform) {template="Rotate Vector::Execute(in Transform,in Vector,out Result)"}
   n10 = RigVMVariableNode() {variable=Belt Rotate Percent}
   Scale = /Script/RigVM.RigVMFunction_MathVectorScale(
    Value=&RigVMFunction_MathTransformRotateVector(Vector=(X=0,Y=0,Z=-1), Transform=@n5.Transform) {template="Rotate Vector::Execute(in Transform,in Vector,out Result)"}.Result
    Factor=&RigVMFunction_MathDoubleSub(A=1, B=RigVMVariableNode() {variable=Belt Rotate Percent}.Value).Result
   ) {is=struct RigVMUnitNode, template="Scale::Execute(in Value,in Factor,out Result)"}
   n11 = &RigVMFunction_MathTransformRotateVector(Vector=(X=1,Y=0,Z=0), Transform=@n8.Transform) {template="Rotate Vector::Execute(in Transform,in Vector,out Result)"}
   n12 = RigVMVariableNode() {variable=Belt Rotate Percent}
   n13 = /Script/ControlRig.RigUnit_AimBoneMath(
    DebugSettings=(bEnabled=False,Scale=10,WorldOffset=(Rotation=(X=0,Y=0,Z=0,W=1),Translation=(X=0,Y=0,Z=0),Scale3D=(X=1,Y=1,Z=1)))
    InputTransform=(Rotation=(X=0,Y=0,Z=0,W=1),Translation=(X=0,Y=0,Z=0),Scale3D=(X=1,Y=1,Z=1))
    Primary=(Weight=1,Axis=(X=0,Y=0,Z=-1),Target=(X=1,Y=0,Z=0),Kind=Direction,Space=(Type=Bone,Name="None"))
    Secondary=(Weight=1,Axis=(X=-1,Y=0,Z=0),Target=(X=1,Y=0,Z=0),Kind=Direction,Space=(Type=Bone,Name="None"))
    Weight=1
    Secondary.Target=&RigVMFunction_MathVectorAdd(
     A=&RigVMFunction_MathVectorScale(Value=@n6.Result, Factor=@n7.Result) {template="Scale::Execute(in Value,in Factor,out Result)"}.Result
     B=&RigVMFunction_MathVectorScale(Value=@n9.Result, Factor=@n10.Value) {template="Scale::Execute(in Value,in Factor,out Result)"}.Result
    ) {template="Add::Execute(in A,in B,out Result)"}.Result
    Primary.Target=&RigVMFunction_MathVectorAdd(
     A=@Scale.Result
     B=&RigVMFunction_MathVectorScale(Value=@n11.Result, Factor=@n12.Value) {template="Scale::Execute(in Value,in Factor,out Result)"}.Result
    ) {template="Add::Execute(in A,in B,out Result)"}.Result
   ) {is=struct RigVMUnitNode}
   > Set Transform = &RigUnit_SetRotation(Item=(Type=Bone,Name="belt_all"), Space=GlobalSpace, Weight=1, bInitial=False, Value=@n13.Result.Rotation)
   RigVMVariableNode(Value=1.5, ExecuteContext=@RigVMFunction_Sequence.D) {variable=Belt Out}
   RigUnit_ItemArray = &RigUnit_ItemArray(Items=((Type=Bone,Name="belt_l"),(Type=Bone,Name="belt_fr_l"),(Type=Bone,Name="belt_r"),(Type=Bone,Name="belt_fr_r"),(Type=Bone,Name="belt_fr_ct")))
   > For_Each = RigVMDispatchNode(Array=@RigUnit_ItemArray.Items) {template="DISPATCH_RigVMDispatch_ArrayIterator(in Array,out Element,out Index,out Count,out Ratio)"}
   Divide = &RigVMFunction_MathDoubleDiv(
    B=2.5
    A=&RigVMFunction_MathDoubleSub(A=RigVMVariableNode() {variable=Squash Distance}.Value, B=RigVMVariableNode() {variable=Original Distance}.Value).Result
   )
   n14 = &RigVMFunction_MathDoubleMul(A=@Divide.Result, B=RigVMVariableNode() {variable=Belt Out}.Value)
   > &RigUnit_OffsetTransformForItem(
    OffsetTransform=(Rotation=(X=0,Y=0,Z=0,W=1),Translation=(X=0,Y=0,Z=0),Scale3D=(X=1,Y=1,Z=1))
    Weight=1
    bPropagateToChildren=True
    Item=@For_Each.Element
    OffsetTransform.Translation.Y=@n14.Result
   )
   n15 = &RigVMFunction_MathDoubleMul(B=0.0187, A=@n14.Result)
   > &RigUnit_OffsetTransformForItem(
    Item=(Type=Bone,Name="shldr_l")
    OffsetTransform=(Rotation=(X=0,Y=0,Z=0,W=1),Translation=(X=0,Y=0,Z=0),Scale3D=(X=1,Y=1,Z=1))
    Weight=1
    bPropagateToChildren=True
    OffsetTransform.Translation.X=@n15.Result
   )
   > &RigUnit_OffsetTransformForItem(
    Item=(Type=Bone,Name="head_l")
    OffsetTransform=(Rotation=(X=0,Y=0,Z=0,W=1),Translation=(X=0,Y=0,Z=0),Scale3D=(X=1,Y=1,Z=1))
    Weight=1
    bPropagateToChildren=True
    OffsetTransform.Translation.X=@n15.Result
   )
   n16 = &RigVMFunction_MathDoubleMul(B=-1, A=@n15.Result)
   > &RigUnit_OffsetTransformForItem(
    Item=(Type=Bone,Name="shldr_r")
    OffsetTransform=(Rotation=(X=0,Y=0,Z=0,W=1),Translation=(X=0,Y=0,Z=0),Scale3D=(X=1,Y=1,Z=1))
    Weight=1
    bPropagateToChildren=True
    OffsetTransform.Translation.X=@n16.Result
   )
   > &RigUnit_OffsetTransformForItem(
    Item=(Type=Bone,Name="head_r")
    OffsetTransform=(Rotation=(X=0,Y=0,Z=0,W=1),Translation=(X=0,Y=0,Z=0),Scale3D=(X=1,Y=1,Z=1))
    Weight=1
    bPropagateToChildren=True
    OffsetTransform.Translation.X=@n16.Result
   )
  ]}
 RigVMModel Construction Graph {nodes [
   n1 = &RigUnit_PrepareForExecution()
   RigVMFunction_ControlFlowBranch = &RigVMFunction_ControlFlowBranch(ExecuteContext=@n1.ExecutePin, Condition=RigVMVariableNode() {variable=Debug Squash Controls}.Value)
   n2 = /Script/ControlRig.RigUnit_HierarchyAddControlTransform(
    InitialValue=(Rotation=(X=0,Y=0,Z=0,W=1),Translation=(X=0,Y=0,Z=0),Scale3D=(X=1,Y=1,Z=1))
    Name=Body
    OffsetSpace=GlobalSpace
    Parent=(Type=Bone,Name="None")
    Settings="(InitialSpace=LocalSpace,bUsePreferredRotationOrder=False,PreferredRotationOrder=YZX,Limits=(LimitTranslationX=(bMinimum=False,bMaximum=False),LimitTranslationY=(bMinimum=False,bMaximum=False),LimitTranslationZ=(bMinimum=False,bMaximum=False),LimitPitch=(b ...(truncated)"
    OffsetTransform=&RigUnit_GetTransform(Item=(Type=Bone,Name="body"), bInitial=False).Transform
    ExecutePin=@RigVMFunction_ControlFlowBranch.True
   ) {
    is=struct RigVMUnitNode
    template="SpawnControl::Execute(in InitialValue,in Settings,in OffsetTransform,in OffsetSpace,in Parent,in Name,out Item)"
   }
   RigUnit_GetTransform = &RigUnit_GetTransform(Item=(Type=Bone,Name="pelvis_MSTR"), bInitial=False)
   > SpawnControl = &RigUnit_HierarchyAddControlTransform(
    InitialValue=(Rotation=(X=0,Y=0,Z=0,W=1),Translation=(X=0,Y=0,Z=0),Scale3D=(X=1,Y=1,Z=1))
    Name=Pelvis
    OffsetSpace=GlobalSpace
    Settings="(InitialSpace=LocalSpace,bUsePreferredRotationOrder=False,PreferredRotationOrder=YZX,Limits=(LimitTranslationX=(bMinimum=False,bMaximum=False),LimitTranslationY=(bMinimum=False,bMaximum=False),LimitTranslationZ=(bMinimum=False,bMaximum=False),LimitPitch=(b ...(truncated)"
    OffsetTransform=@RigUnit_GetTransform.Transform
    Parent=@n2.Item
   ) {template="SpawnControl::Execute(in InitialValue,in Settings,in OffsetTransform,in OffsetSpace,in Parent,in Name,out Item)"}
  ]}
]
hierarchy [
 root {loc=<0 0 0>, rot=<0 0 0 1>}
 pack {parent=root, loc=<0 -0.7316 0>, rot=<0 0 0 1>}
 chicken_MSTR {parent=pack, loc=<0 0 0>, rot=<0 0 0 1>}
 chicken {parent=chicken_MSTR, loc=<0 0 0>, rot=<0 0 0 1>}
 body {parent=chicken, loc=<-0.0438 -10.2104 -11.6688>, rot=<0 0 0 1>}
 pelvis_MSTR {parent=body, loc=<0 0 0>, rot=<-0.5 -0.5 0.5 -0.5>}
 pelvis_01 {parent=pelvis_MSTR, loc=<0 0 0>, rot=<0 0 0 1>}
 tail_MSTR {parent=pelvis_01, loc=<0 0 0>, rot=<0 0 0 1>}
 tail_01 {parent=tail_MSTR, loc=<0 0 0>, rot=<0 0 0 1>}
 tail_02 {parent=tail_01, loc=<10.0702 1.1777 -1.0576>, rot=<0.6638 0.2436 0.1322 -0.6947>}
 tail_03 {parent=tail_02, loc=<5.0345 -1.078 2.633>, rot=<-0.0485 0.219 0.2398 -0.9446>}
 leg_l_MSTR {parent=pelvis_01, loc=<0 0 0>, rot=<0 0 0 1>}
 leg_l_01 {parent=leg_l_MSTR, loc=<0 0 0>, rot=<0 0 0 1>}
 leg_l_02 {parent=leg_l_01, loc=<4.4767 -5.9619 11.7385>, rot=<-0.1996 0.441 0.1013 -0.8691>}
 leg_l_03 {parent=leg_l_02, loc=<2.9531 0 0>, rot=<0 0 0 1>}
 leg_r_MSTR {parent=pelvis_01, loc=<0 0 0>, rot=<0 0 0 1>}
 leg_r_01 {parent=leg_r_MSTR, loc=<0 0 0>, rot=<0 0 0 1>}
 leg_r_02 {parent=leg_r_01, loc=<4.4767 6.0495 11.7385>, rot=<0.1996 0.441 -0.1013 -0.8691>}
 leg_r_03 {parent=leg_r_02, loc=<2.9531 0 0>, rot=<0 0 0 1>}
 belt_MSTR {parent=body, loc=<0 0 0>, rot=<0 0 0 1>}
 belt_all {parent=belt_MSTR, loc=<0 0 0>, rot=<0 0 0 1>}
 belt_l {parent=belt_all, loc=<-12.134 0.6834 5.262>, rot=<0 0 0.7071 -0.7071>}
 belt_fr_l {parent=belt_all, loc=<-9.6434 -8.667 5.2619>, rot=<0 0 0.3523 -0.9359>}
 holster_l_MSTR {parent=belt_fr_l, loc=<-3.383 -0.443 -2.116>, rot=<-0.3521 0.1551 0.8999 0.2051>}
 holster_l {parent=holster_l_MSTR, loc=<0 0 0>, rot=<0 0 0 1>}
 belt_ba_l {parent=belt_all, loc=<-7.0161 9.7545 4.5195>, rot=<0 0 0.9921 -0.1251>}
 belt_r {parent=belt_all, loc=<12.1343 0.6834 5.262>, rot=<0 0 -0.7071 -0.7071>}
 belt_fr_r {parent=belt_all, loc=<9.7049 -8.7347 5.3421>, rot=<0 0 -0.3523 -0.9359>}
 holster_r_MSTR {parent=belt_fr_r, loc=<3.4472 -0.4093 -2.1962>, rot=<0.152 -0.3238 -0.2097 -0.91>}
 holster_r {parent=holster_r_MSTR, loc=<0 0 0>, rot=<0 0 0 1>}
 belt_ba_r {parent=belt_all, loc=<7.1038 9.7545 4.5195>, rot=<0 0 0.981 0.1942>}
 belt_fr_ct {parent=belt_all, loc=<0.0425 -12.2315 5.5163>, rot=<0 0 0 1>}
 head_MSTR {parent=body, loc=<0.0438 0.6393 12.3778>, rot=<0 0 0 1>}
 head_01 {parent=head_MSTR, loc=<0 0 0>, rot=<0 0 0 1>}
 head_02 {parent=head_01, loc=<0 0 5.2936>, rot=<0 0 0 1>}
 egg_MSTR {parent=head_02, loc=<-2.7568 3.5844 8.7775>, rot=<0 0 0 1>}
 egg {parent=egg_MSTR, loc=<0 0 0>, rot=<0 0 0 1>}
 comb_MSTR {parent=head_02, loc=<0.1908 -4.3374 5.3346>, rot=<0.1415 0.2732 0.4376 -0.8449>}
 comb_01 {parent=comb_MSTR, loc=<0 0 0>, rot=<0 0 0 1>}
 comb_02 {parent=comb_01, loc=<3.4575 0 0>, rot=<0.2266 -0.2924 -0.2836 -0.8847>}
 glasses_MSTR {parent=head_02, loc=<0 -3.6976 2.2031>, rot=<0 0 0.7071 -0.7071>}
 glasses {parent=glasses_MSTR, loc=<0 0 0>, rot=<0 0 0 1>}
 head_l {parent=head_01, loc=<-7.4085 0.4447 4.2693>, rot=<0 0 0 1>}
 head_r {parent=head_01, loc=<7.408 0.4447 4.2693>, rot=<0 0 0 1>}
 head_fr {parent=head_01, loc=<0 -8.1518 0.6943>, rot=<0 0 0 1>}
 beak {parent=head_01, loc=<0 -6.7173 4.8271>, rot=<0 0 0 1>}
 shldr_l {parent=body, loc=<-9.5022 0.5736 13.1163>, rot=<0 0 0 1>}
 shldr_r {parent=body, loc=<9.5898 0.5736 13.1163>, rot=<0 0 0 1>}
 arm_l_MSTR {parent=body, loc=<-10.3202 0.8555 10.1151>, rot=<-0.3083 0.0014 0.9513 0.0044>}
 arm_l_01 {parent=arm_l_MSTR, loc=<0 0 0>, rot=<0 0 0 1>}
 arm_l_02 {parent=arm_l_01, loc=<7.6259 0 0>, rot=<0 0 0 1>}
 arm_r_MSTR {parent=body, loc=<10.4078 0.8555 10.1151>, rot=<0.0014 -0.3083 -0.0044 -0.9513>}
 arm_r_01 {parent=arm_r_MSTR, loc=<0 0 0>, rot=<0 0 0 1>}
 arm_r_02 {parent=arm_r_01, loc=<7.626 0 0>, rot=<0 0 0 1>}
]
variables [
 Belt Aim {type=FVector, default=(X=0,Y=0,Z=0)}
 Belt Out {type=double, default=0.5}
 Belt Push Percent {type=double, default=0.5}
 Belt Rotate Percent {type=double, default=0.5}
 Belt Secon {type=FVector, default=(X=0,Y=0,Z=0)}
 Debug Squash Controls {type=bool, default=False}
 Debug Squash Values {type=bool, default=False}
 Original Distance {type=double, default=10}
 Squash Distance {type=double, default=0}
 Squash End A {type=FTransform, default=(Rotation=(X=0,Y=0,Z=0,W=1),Translation=(X=0,Y=0,Z=0),Scale3D=(X=1,Y=1,Z=1))}
 Squash End B {type=FTransform, default=(Rotation=(X=0,Y=0,Z=0,W=1),Translation=(X=0,Y=0,Z=0),Scale3D=(X=1,Y=1,Z=1))}
]
