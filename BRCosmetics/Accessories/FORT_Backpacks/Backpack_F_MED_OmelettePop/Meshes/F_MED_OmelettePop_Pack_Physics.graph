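# Physics asset for the OmelettePop backpack cosmetic, defining simple bodies and joints for secondary motion of its small character figure.
# Separate chains hang from pelvis, head and left and right arm MSTR bodies, driving pelvis with legs, head with two-part comb, and two-segment arms.
# Joints have narrow swing limits, widest at the legs, with twist locked on the comb.
# Two holster boxes have no constraints.
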PhysicsAsset /BRCosmetics/Accessories/FORT_Backpacks/Backpack_F_MED_OmelettePop/Meshes/F_MED_OmelettePop_Pack_Physics
bodies {
 arm_l_01 [{shapeType=Box, extentX=10.5, extentY=12.5, extentZ=3.5, shapeName=arm_l_01_box, center=<1.476 0.7081 0.9997>}]
 arm_l_02 [{shapeType=Box, extentX=9.5, extentY=9.5, extentZ=3.5, shapeName=arm_l_02_box, center=<3.55 0.5481 0>}]
 arm_l_MSTR [{shapeType=Sphere, radius=0.1, shapeName=arm_l_MSTR_sphere, center=<2.1335 -0.3081 -1.359>}]
 arm_r_01 [{shapeType=Box, extentX=10.5, extentY=12.5, extentZ=3.5, shapeName=arm_l_01_box, center=<1.476 -0.7081 0.9997>}]
 arm_r_02 [{shapeType=Box, extentX=9.5, extentY=9.5, extentZ=3.5, shapeName=arm_l_02_box, center=<3.55 -0.5481 0>}]
 arm_r_MSTR [{shapeType=Sphere, radius=0.1, shapeName=arm_r_MSTR_sphere, center=<2.3487 0.1503 -1.9832>}]
 comb_01 [{shapeType=Capsule, length=3.458, radius=0.33, shapeName=comb_01_capsule, center=<1.729 0 0>, rotation=<rot 90 0 0>}]
 comb_02 [{shapeType=Capsule, length=3, radius=0.1, shapeName=comb_02_capsule, center=<1.5 0 0>, rotation=<rot 90 0 0>}]
 head_01 [{shapeType=Capsule, length=3, radius=1.5, shapeName=head_01_capsule, center=<0 0 2.6078>}]
 head_MSTR [{shapeType=Capsule, length=0.1, radius=0.505, shapeName=head_MSTR_capsule, center=<0.0201 -7.8645 -0.5307>, rotation=<rot 0 90.2573 -89.9375>}]
 holster_l [{shapeType=Box, extentX=14, extentY=3, extentZ=10, shapeName=holster_r_box, center=<1.1192 1.952 -2.6897>, rotation=<rot -5.6248 0 -0.0001>}]
 holster_r [{shapeType=Box, extentX=14, extentY=3, extentZ=10, shapeName=holster_r_box, center=<1.1192 -1.952 -2.6897>, rotation=<rot -5.6248 0 -0.0001>}]
 leg_l_02 [{shapeType=Capsule, length=2.953, radius=0.5, shapeName=leg_r_02_capsule, center=<1.4765 0 0>, rotation=<rot 90 0 0>}]
 leg_l_03 [{shapeType=Capsule, length=2.953, radius=0.5, shapeName=leg_r_03_capsule, center=<1.4765 0 0>, rotation=<rot 90 0 0>}]
 leg_r_02 [{shapeType=Capsule, length=2.953, radius=0.5, shapeName=leg_r_02_capsule, center=<1.4765 0 0>, rotation=<rot 90 0 0>}]
 leg_r_03 [{shapeType=Capsule, length=2.953, radius=0.5, shapeName=leg_r_03_capsule, center=<1.4765 0 0>, rotation=<rot 90 0 0>}]
 pelvis_01 [{shapeType=Capsule, length=10.005, radius=1, shapeName=pelvis_01_capsule, center=<5.2132 0 0>, rotation=<rot 90 0 0>}]
 pelvis_MSTR [{shapeType=Capsule, length=0.505, radius=0.505, shapeName=pelvis_MSTR_capsule}]
}
constraints [
 {bone1Name=pelvis_01, bone2Name=pelvis_MSTR, swing1LimitDegrees=6, swing2LimitDegrees=10, twistLimitDegrees=0}
 {bone1Name=head_01, bone2Name=head_MSTR, swing1LimitDegrees=3, swing2LimitDegrees=12, twistLimitDegrees=5}
 {bone1Name=comb_01, bone2Name=head_01, swing1LimitDegrees=2, swing2LimitDegrees=4, twistLimitDegrees=0, twistMotion=Locked}
 {bone1Name=comb_02, bone2Name=comb_01, swing1LimitDegrees=2, swing2LimitDegrees=4, twistLimitDegrees=0, twistMotion=Locked}
 {bone1Name=leg_r_02, bone2Name=pelvis_01, swing1LimitDegrees=15, swing2LimitDegrees=15, twistLimitDegrees=0}
 {bone1Name=leg_r_03, bone2Name=leg_r_02, swing1LimitDegrees=15, swing2LimitDegrees=15, twistLimitDegrees=0}
 {bone1Name=leg_l_03, bone2Name=leg_l_02, swing1LimitDegrees=15, swing2LimitDegrees=15, twistLimitDegrees=0}
 {bone1Name=leg_l_02, bone2Name=pelvis_01, swing1LimitDegrees=15, swing2LimitDegrees=15, twistLimitDegrees=0}
 {bone1Name=arm_r_02, bone2Name=arm_r_01, swing1LimitDegrees=5, swing2LimitDegrees=10, twistLimitDegrees=0}
 {bone1Name=arm_l_02, bone2Name=arm_l_01, swing1LimitDegrees=5, swing2LimitDegrees=10, twistLimitDegrees=0}
 {bone1Name=arm_l_01, bone2Name=arm_l_MSTR, swing1LimitDegrees=5, swing2LimitDegrees=10, twistLimitDegrees=0}
 {bone1Name=arm_r_01, bone2Name=arm_r_MSTR, swing1LimitDegrees=5, swing2LimitDegrees=10, twistLimitDegrees=0}
]
