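# Animation Blueprint for the OmelettePop chicken backpack cosmetic, adding falling and frontend poses plus secondary motion.
# Each update flags falling without skydiving and frontend display, which blend in additive bone-space offsets.
# Falling offsets shift egg, glasses, body, shoulders and arms before a RigidBody simulation based on the chicken bone, then tilt the comb afterward.
# Frontend offsets then rotate the arms slightly, and a Control Rig finishes the pose.
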AnimBlueprint /BRCosmetics/Accessories/FORT_Backpacks/Backpack_F_MED_OmelettePop/Meshes/F_MED_OmelettePop_Pack_AnimBP
target_skeleton=&/F_MED_OmelettePop_Pack_Skeleton
graphs [
 AnimGraph {nodes [
   n1 = Get isFallingNotSkydiving() -> (isFallingNotSkydiving:bool)
   n2 = "Transform (Modify) Bone - Bone: egg_MSTR"!(
    ComponentPose=Local To Component(LocalPose=Input Pose())
    bAlphaBoolEnabled:bool=@n1
    Translation="0,-4,15"
    Rotation="0,0,-15"
    Scale="1,1,1"
   ) -> (Pose) {
    alphaInputType=Bool
    rotationMode=BMM_Additive
    rotationSpace=BCS_BoneSpace
    translationMode=BMM_Additive
    translationSpace=BCS_BoneSpace
    alphaBoolBlend {blendInTime=1.25, blendOption=Linear, blendOutTime=0.1}
    boneToModify {boneName=egg_MSTR}
    rotation {roll=-15}
    translation {y=-4, z=15}
   }
   n3 = "Transform (Modify) Bone - Bone: glasses_MSTR"!(ComponentPose=@n2, bAlphaBoolEnabled:bool=@n1, Translation="1,0,0.125", Rotation="45,0,0", Scale="1,1,1") -> (Pose) {
    alphaInputType=Bool
    rotationMode=BMM_Additive
    rotationSpace=BCS_BoneSpace
    translationMode=BMM_Additive
    translationSpace=BCS_BoneSpace
    alphaBoolBlend {blendInTime=0.5, blendOption=Linear, blendOutTime=0.1}
    boneToModify {boneName=glasses_MSTR}
    rotation {pitch=45}
    translation {x=1, z=0.125}
   }
   n4 = "Transform (Modify) Bone - Bone: pelvis_MSTR"!(ComponentPose=@n3, bAlphaBoolEnabled:bool=@n1, Translation="-1,0,0", Rotation="0,0,0", Scale="1,1,1") -> (Pose) {
    alphaInputType=Bool
    rotationMode=BMM_Additive
    rotationSpace=BCS_BoneSpace
    translationMode=BMM_Additive
    translationSpace=BCS_BoneSpace
    alphaBoolBlend {blendInTime=0.5, blendOption=Linear, blendOutTime=0.1}
    boneToModify {boneName=pelvis_MSTR}
    translation {x=-1}
   }
   n5 = "Transform (Modify) Bone - Bone: head_MSTR"!(ComponentPose=@n4, bAlphaBoolEnabled:bool=@n1, Translation="0,0,1.5", Rotation="0,0,0", Scale="1,1,1") -> (Pose) {
    alphaInputType=Bool
    rotationMode=BMM_Additive
    rotationSpace=BCS_BoneSpace
    translationMode=BMM_Additive
    translationSpace=BCS_BoneSpace
    alphaBoolBlend {blendInTime=0.5, blendOption=Linear, blendOutTime=0.1}
    boneToModify {boneName=head_MSTR}
    translation {z=1.5}
   }
   n6 = "Transform (Modify) Bone - Bone: shldr_l"!(ComponentPose=@n5, bAlphaBoolEnabled:bool=@n1, Translation="0,0,1", Rotation="0,0,0", Scale="1,1,1") -> (Pose) {
    alphaInputType=Bool
    rotationMode=BMM_Additive
    rotationSpace=BCS_BoneSpace
    translationMode=BMM_Additive
    translationSpace=BCS_BoneSpace
    alphaBoolBlend {blendInTime=0.5, blendOption=Linear, blendOutTime=0.1}
    boneToModify {boneName=shldr_l}
    translation {z=1}
   }
   n7 = "Transform (Modify) Bone - Bone: shldr_r"!(ComponentPose=@n6, bAlphaBoolEnabled:bool=@n1, Translation="0,0,1", Rotation="0,0,0", Scale="1,1,1") -> (Pose) {
    alphaInputType=Bool
    rotationMode=BMM_Additive
    rotationSpace=BCS_BoneSpace
    translationMode=BMM_Additive
    translationSpace=BCS_BoneSpace
    alphaBoolBlend {blendInTime=0.5, blendOption=Linear, blendOutTime=0.1}
    boneToModify {boneName=shldr_r}
    translation {z=1}
   }
   n8 = "Transform (Modify) Bone - Bone: arm_l_MSTR"!(ComponentPose=@n7, bAlphaBoolEnabled:bool=@n1, Translation="0,0,0", Rotation="45,0,0", Scale="1,1,1") -> (Pose) {
    alphaInputType=Bool
    rotationMode=BMM_Additive
    rotationSpace=BCS_BoneSpace
    translationMode=BMM_Additive
    translationSpace=BCS_BoneSpace
    alphaBoolBlend {blendInTime=0.5, blendOption=Linear, blendOutTime=0.1}
    boneToModify {boneName=arm_l_MSTR}
    rotation {pitch=45}
   }
   n9 = "Transform (Modify) Bone - Bone: arm_r_MSTR"!(ComponentPose=@n8, bAlphaBoolEnabled:bool=@n1, Translation="0,0,0", Rotation="45,0,0", Scale="1,1,1") -> (Pose) {
    alphaInputType=Bool
    rotationMode=BMM_Additive
    rotationSpace=BCS_BoneSpace
    translationMode=BMM_Additive
    translationSpace=BCS_BoneSpace
    alphaBoolBlend {blendInTime=0.5, blendOption=Linear, blendOutTime=0.1}
    boneToModify {boneName=arm_r_MSTR}
    rotation {pitch=45}
   }
   n10 = RigidBody(ComponentPose=@n9, ExternalForce="0,0,-980") {
    bClampLinearTranslationLimitToRefPose=true
    bOverrideWorldGravity=true
    overridePhysicsAsset=&/F_MED_OmelettePop_Pack_Physics
    simulationSpace=BaseBoneSpace
    alphaBoolBlend {blendOption=Linear}
    baseBoneRef {boneName=chicken}
    componentAppliedLinearAccClamp=<3000 3000 3000>
    componentLinearAccScale=<0.4 0.3 0.3>
    componentLinearVelScale=<0.4 0.3 0.3>
    overrideWorldGravity {z=-980}
   }
   n11 = "Transform (Modify) Bone - Bone: comb_MSTR"!(ComponentPose=@n10, bAlphaBoolEnabled:bool=@n1, Translation="0,0,0.1", Rotation="45,0,0", Scale="1,1,1") -> (Pose) {
    alphaInputType=Bool
    rotationMode=BMM_Additive
    rotationSpace=BCS_BoneSpace
    translationMode=BMM_Additive
    translationSpace=BCS_BoneSpace
    alphaBoolBlend {blendInTime=0.5, blendOption=Linear, blendOutTime=0.1}
    boneToModify {boneName=comb_MSTR}
    rotation {pitch=45}
    translation {z=0.1}
   }
   n12 = "Transform (Modify) Bone - Bone: comb_02"!(ComponentPose=@n11, bAlphaBoolEnabled:bool=@n1, Translation="0,0,0", Rotation="22.5,0,0", Scale="1,1,1") -> (Pose) {
    alphaInputType=Bool
    rotationMode=BMM_Additive
    rotationSpace=BCS_BoneSpace
    translationMode=BMM_Additive
    translationSpace=BCS_BoneSpace
    alphaBoolBlend {blendInTime=0.5, blendOption=Linear, blendOutTime=0.1}
    boneToModify {boneName=comb_02}
    rotation {pitch=22.5}
   }
   n13 = Get isFrontEnd() -> (isFrontEnd:bool)
   n14 = "Transform (Modify) Bone - Bone: arm_r_MSTR"(ComponentPose=@n12, bAlphaBoolEnabled=@n13, Translation="0,0,0", Rotation="7.5,0,0", Scale="1,1,1") {
    alphaInputType=Bool
    rotationMode=BMM_Additive
    rotationSpace=BCS_BoneSpace
    translationMode=BMM_Additive
    translationSpace=BCS_BoneSpace
    alphaBoolBlend {blendInTime=0.5, blendOption=Linear, blendOutTime=0.1}
    boneToModify {boneName=arm_r_MSTR}
    rotation {pitch=7.5}
   }
   n15 = "Transform (Modify) Bone - Bone: arm_l_MSTR"(ComponentPose=@n14, bAlphaBoolEnabled=@n13, Translation="0,0,0", Rotation="7.5,0,0", Scale="1,1,1") {
    alphaInputType=Bool
    rotationMode=BMM_Additive
    rotationSpace=BCS_BoneSpace
    translationMode=BMM_Additive
    translationSpace=BCS_BoneSpace
    alphaBoolBlend {blendInTime=0.5, blendOption=Linear, blendOutTime=0.1}
    boneToModify {boneName=arm_l_MSTR}
    rotation {pitch=7.5}
   }
   n16 = "Transform (Modify) Bone - Bone: arm_r_02"!(ComponentPose=@n15, bAlphaBoolEnabled:bool=@n13, Translation="0,0,0", Rotation="3.5,0,0", Scale="1,1,1") -> (Pose) {
    alphaInputType=Bool
    rotationMode=BMM_Additive
    rotationSpace=BCS_BoneSpace
    translationMode=BMM_Additive
    translationSpace=BCS_BoneSpace
    alphaBoolBlend {blendInTime=0.5, blendOption=Linear, blendOutTime=0.1}
    boneToModify {boneName=arm_r_02}
    rotation {pitch=3.5}
   }
   n17 = "Transform (Modify) Bone - Bone: arm_l_02"!(ComponentPose=@n16, bAlphaBoolEnabled:bool=@n13, Translation="0,0,0", Rotation="3.5,0,0", Scale="1,1,1") -> (Pose) {
    alphaInputType=Bool
    rotationMode=BMM_Additive
    rotationSpace=BCS_BoneSpace
    translationMode=BMM_Additive
    translationSpace=BCS_BoneSpace
    alphaBoolBlend {blendInTime=0.5, blendOption=Linear, blendOutTime=0.1}
    boneToModify {boneName=arm_l_02}
    rotation {pitch=3.5}
   }
   n18 = Control Rig(Source=Component To Local(ComponentPose=@n17)) {
    alphaBoolBlend {blendOption=Linear}
    controlRigAssetReference {blueprintRigClass=&/F_MED_OmelettePop_Pack_CtrlRig.F_MED_OmelettePop_Pack_CtrlRig_C}
   }
   Output Pose(Result=@n18) {layerGroup=None}
  ]}
 EventGraph {nodes [
   Event BlueprintUpdateAnimation()
   > n1 = Sequence()
   n2 = Get bIsSkyDiving()
   > n3 = Branch(Condition=AND Boolean(A=Get bIsFalling(), B=NOT Boolean(A=@n2)))
   > Set isFallingNotSkydiving(isFallingNotSkydiving:bool=true) -> (Output_Get:bool)
   Set isFallingNotSkydiving(execute=@n3.else, isFallingNotSkydiving=false)
   Set isFrontEnd(execute=@n1.then_1, isFrontEnd:bool=IsCosmeticObjectDisplayedInFrontendContext!(CosmeticObject:Object Reference) -> (ReturnValue:bool)) -> (Output_Get:bool)
   TryGetPawnOwner()
  ]}
]
variables [
 isFallingNotSkydiving {type=boolean}
 isFrontEnd {type=boolean}
]
